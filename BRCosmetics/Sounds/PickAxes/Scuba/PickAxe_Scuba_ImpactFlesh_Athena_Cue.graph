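# Sound cue for the Scuba pickaxe's flesh impact in Athena, played when the pickaxe hits an enemy.
# A short fixed delay feeds a random node that picks one of six equally weighted enemy impact waves.
# It uses shared pickaxe impact attenuation, a property sheet, the impacts sound class and a melee weapon concurrency set, and is primed on load.
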SoundCue /BRCosmetics/Sounds/PickAxes/Scuba/PickAxe_Scuba_ImpactFlesh_Athena_Cue
root=Delay_1
nodes [
 {id=Delay_1, children [Random_2], properties {delayMax=0.1, delayMin=0.1}}
 {
  id=Random_2
  children [WavePlayer_11, WavePlayer_12, WavePlayer_13, WavePlayer_14, WavePlayer_15, WavePlayer_16]
  properties {weights [1,1,1,1,1,1]}
 }
 {id=WavePlayer_11, properties {soundWaveAssetPtr=&/PickAxe_Scuba_EnemyImpact_04}}
 {id=WavePlayer_12, properties {soundWaveAssetPtr=&/PickAxe_Scuba_EnemyImpact_06-glued}}
 {id=WavePlayer_13, properties {soundWaveAssetPtr=&/PickAxe_Scuba_EnemyImpact_01}}
 {id=WavePlayer_14, properties {soundWaveAssetPtr=&/PickAxe_Scuba_EnemyImpact_02}}
 {id=WavePlayer_15, properties {soundWaveAssetPtr=&/PickAxe_Scuba_EnemyImpact_03}}
 {id=WavePlayer_16, properties {soundWaveAssetPtr=&/PickAxe_Scuba_EnemyImpact_05-glued}}
]
properties {
 attenuationSettings=/Game/Sounds/Attenuation/Fort_Pickaxe_Impact_Att
 audioPropertiesSheet=/BRCosmetics/Sounds/Settings/PropertySheets/Pickaxe/APS_Cosmetic_Pickaxe_Impact
 bPrimeOnLoad=true
 maxDistance=4300
 soundClassObject=&Fort_Impacts
 volumeMultiplier=0.525
 attenuationOverrides {
  bAttenuateWithLPF=true
  distanceAlgorithm=Logarithmic
  falloffDistance=800
  lPFRadiusMax=800
  lPFRadiusMin=400
  attenuationShapeExtents {x=300}
 }
 concurrencyOverrides {bLimitToOwner=true, maxCount=2, resolutionRule=StopOldest}
 concurrencySet [&Player_Melee_Weapon_Conc]
}
